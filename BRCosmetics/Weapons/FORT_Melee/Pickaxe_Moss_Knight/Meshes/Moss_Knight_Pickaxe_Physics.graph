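# Physics asset for the Moss Knight two-handed pickaxe, defining bodies and constraints for its dangling vines, leaves and flowers.
# Capsule vines form central, side and small chains rooted on the box handle, alongside a single main vine base, with tight angular limits.
# Box flowers hang off the handle, while looser-swinging leaves attach to the handle or vine segments.
# An extra angle-locked constraint ties the handle to the central chain's third segment.
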PhysicsAsset /BRCosmetics/Weapons/FORT_Melee/Pickaxe_Moss_Knight/Meshes/Moss_Knight_Pickaxe_Physics
bodies {
 Melee_TwoHanded_Handle [{shapeType=Box, extentX=8.2772, extentY=2.2733, extentZ=51.3351, center=<0.0517 -0.1261 32.3125>, rotation=<rot -0.3485 -90 0.0748>}]
 dyn_central_vine_01 [{shapeType=Capsule, length=2.1032, radius=1, center=<2.0621 0 -0.0001>, rotation=<rot 90 0 0>}]
 dyn_central_vine_02 [{shapeType=Capsule, length=2.1032, radius=1, center=<2.2533 0 -0.0002>, rotation=<rot 90 0 0>}]
 dyn_central_vine_03 [{shapeType=Capsule, length=3.4089, radius=0.7982, center=<2.6504 0 -0.0002>, rotation=<rot 90 0 0>}]
 dyn_central_vine_04 [{shapeType=Capsule, length=3.2193, radius=1, center=<2.7782 0 -0.0001>, rotation=<rot 90 0 0>}]
 dyn_central_vine_05 [{shapeType=Capsule, length=2.1032, radius=1, center=<2.0621 0 -0.0001>, rotation=<rot 90 0 0>}]
 dyn_central_vine_06 [{shapeType=Capsule, length=2.1032, radius=1, center=<2.0621 0 -0.0001>, rotation=<rot 90 0 0>}]
 dyn_flower_01 [{shapeType=Box, extentX=0.7451, extentY=3.3816, extentZ=3.1826, center=<0.818 0 0>}]
 dyn_flower_02 [{shapeType=Box, extentX=0.7451, extentY=3.3816, extentZ=3.1826, center=<0.818 0 0>}]
 dyn_flower_03 [{shapeType=Box, extentX=0.7451, extentY=3.3816, extentZ=3.1826, center=<0.818 0 0>}]
 dyn_flower_04 [{shapeType=Box, extentX=0.7451, extentY=3.3816, extentZ=3.1826, center=<0.818 0 0>}]
 dyn_flower_05 [{shapeType=Box, extentX=0.7451, extentY=3.3816, extentZ=3.1826, center=<0.818 0 0>}]
 dyn_flower_06 [{shapeType=Box, extentX=0.7451, extentY=3.3816, extentZ=3.1826, center=<0.818 0 0>}]
 dyn_flower_07 [{shapeType=Box, extentX=0.7451, extentY=3.3816, extentZ=3.1826, center=<0.818 0 0>}]
 dyn_leaf_01 [{shapeType=Box, extentX=2.7373, extentY=2, extentZ=1.0712, center=<1.5639 0 0>}]
 dyn_leaf_02 [{shapeType=Box, extentX=2.7373, extentY=2, extentZ=1.0712, center=<1.5639 0 0>}]
 dyn_leaf_03 [{shapeType=Box, extentX=2.7373, extentY=2, extentZ=1.0712, center=<1.5639 0 0>}]
 dyn_leaf_04 [{shapeType=Box, extentX=3.9295, extentY=2, extentZ=1.5795, center=<2.3393 0 0>}]
 dyn_leaf_05 [{shapeType=Box, extentX=2.7373, extentY=2, extentZ=1.0712, center=<1.5639 0 0>}]
 dyn_leaf_06 [{shapeType=Box, extentX=2.7373, extentY=2, extentZ=1.0712, center=<1.5639 0 0>}]
 dyn_leaf_07 [{shapeType=Box, extentX=2.7373, extentY=2, extentZ=1.0712, center=<1.5639 0 0>}]
 dyn_leaf_08 [{shapeType=Box, extentX=2.7373, extentY=2, extentZ=1.0712, center=<1.5639 0 0>}]
 dyn_leaf_09 [{shapeType=Box, extentX=2.7373, extentY=2, extentZ=1.0712, center=<1.5639 0 0>}]
 dyn_leaf_10 [{shapeType=Box, extentX=3.4562, extentY=2, extentZ=1.5733, center=<1.9875 0 0>}]
 dyn_leaf_11 [{shapeType=Box, extentX=2.7373, extentY=2, extentZ=1.0712, center=<1.5639 0 0>}]
 dyn_leaf_12 [{shapeType=Box, extentX=2.7373, extentY=2, extentZ=1.0712, center=<1.5639 0 0>}]
 dyn_main_vine_base [{shapeType=Capsule, length=1.4, radius=0.6, center=<0.7212 0 0>, rotation=<rot 90 0 0>}]
 dyn_side_vine_01 [{shapeType=Capsule, length=1.7647, radius=1, center=<1.816 0 -0.0001>, rotation=<rot 90 0 0>}]
 dyn_side_vine_02 [{shapeType=Capsule, length=1, radius=1, center=<1.4941 0 -0.0001>, rotation=<rot 90 0 0>}]
 dyn_side_vine_03 [{shapeType=Capsule, length=3.1632, radius=1, center=<2.7904 0 -0.0001>, rotation=<rot 90 0 0>}]
 dyn_side_vine_04 [{shapeType=Capsule, length=1.1259, radius=1, center=<1.4941 0 -0.0001>, rotation=<rot 90 0 0>}]
 dyn_side_vine_05 [{shapeType=Capsule, length=1, radius=1, center=<1.4941 0 -0.0001>, rotation=<rot 90 0 0>}]
 dyn_small_vine_01 [{shapeType=Capsule, length=1, radius=1, center=<1.5455 0 0>, rotation=<rot 90 0 0>}]
 dyn_small_vine_02 [{shapeType=Capsule, length=1, radius=1, center=<1.5455 0 0>, rotation=<rot 90 0 0>}]
 dyn_small_vine_03 [{shapeType=Capsule, length=1, radius=1, center=<1.5455 0 0>, rotation=<rot 90 0 0>}]
}
constraints [
 {bone1Name=dyn_flower_07, bone2Name=Melee_TwoHanded_Handle, swing1LimitDegrees=20, swing2LimitDegrees=20, twistLimitDegrees=20}
 {bone1Name=dyn_flower_03, bone2Name=Melee_TwoHanded_Handle, swing1LimitDegrees=20, swing2LimitDegrees=20, twistLimitDegrees=20}
 {bone1Name=dyn_flower_02, bone2Name=Melee_TwoHanded_Handle, swing1LimitDegrees=20, swing2LimitDegrees=20, twistLimitDegrees=20}
 {bone1Name=dyn_flower_01, bone2Name=Melee_TwoHanded_Handle, swing1LimitDegrees=20, swing2LimitDegrees=20, twistLimitDegrees=20}
 {bone1Name=dyn_flower_04, bone2Name=Melee_TwoHanded_Handle, swing1LimitDegrees=20, swing2LimitDegrees=20, twistLimitDegrees=20}
 {bone1Name=dyn_flower_06, bone2Name=Melee_TwoHanded_Handle, swing1LimitDegrees=20, swing2LimitDegrees=20, twistLimitDegrees=20}
 {bone1Name=dyn_flower_05, bone2Name=Melee_TwoHanded_Handle, swing1LimitDegrees=20, swing2LimitDegrees=20, twistLimitDegrees=20}
 {bone1Name=dyn_main_vine_base, bone2Name=Melee_TwoHanded_Handle, swing1LimitDegrees=7, swing2LimitDegrees=7, twistLimitDegrees=7}
 {bone1Name=dyn_side_vine_01, bone2Name=Melee_TwoHanded_Handle, swing1LimitDegrees=5, swing2LimitDegrees=5, twistLimitDegrees=5}
 {bone1Name=dyn_side_vine_02, bone2Name=dyn_side_vine_01, swing1LimitDegrees=5, swing2LimitDegrees=5, twistLimitDegrees=5}
 {bone1Name=dyn_side_vine_03, bone2Name=dyn_side_vine_02, swing1LimitDegrees=5, swing2LimitDegrees=5, twistLimitDegrees=5}
 {bone1Name=dyn_side_vine_04, bone2Name=dyn_side_vine_03, swing1LimitDegrees=5, swing2LimitDegrees=5, twistLimitDegrees=5}
 {bone1Name=dyn_side_vine_05, bone2Name=dyn_side_vine_04, swing1LimitDegrees=5, swing2LimitDegrees=5, twistLimitDegrees=5}
 {bone1Name=dyn_small_vine_01, bone2Name=Melee_TwoHanded_Handle, swing1LimitDegrees=10, swing2LimitDegrees=10, twistLimitDegrees=10}
 {bone1Name=dyn_small_vine_02, bone2Name=dyn_small_vine_01, swing1LimitDegrees=10, swing2LimitDegrees=10, twistLimitDegrees=10}
 {bone1Name=dyn_small_vine_03, bone2Name=dyn_small_vine_02, swing1LimitDegrees=10, swing2LimitDegrees=10, twistLimitDegrees=10}
 {bone1Name=dyn_central_vine_01, bone2Name=Melee_TwoHanded_Handle, swing1LimitDegrees=10, swing2LimitDegrees=10, twistLimitDegrees=10}
 {bone1Name=dyn_central_vine_02, bone2Name=dyn_central_vine_01, swing1LimitDegrees=10, swing2LimitDegrees=10, twistLimitDegrees=10}
 {bone1Name=dyn_central_vine_03, bone2Name=dyn_central_vine_02, swing1LimitDegrees=10, swing2LimitDegrees=10, twistLimitDegrees=10}
 {bone1Name=dyn_central_vine_04, bone2Name=dyn_central_vine_03, swing1LimitDegrees=10, swing2LimitDegrees=10, twistLimitDegrees=10}
 {bone1Name=dyn_central_vine_05, bone2Name=dyn_central_vine_04, swing1LimitDegrees=10, swing2LimitDegrees=10, twistLimitDegrees=10}
 {bone1Name=dyn_central_vine_06, bone2Name=dyn_central_vine_05, swing1LimitDegrees=10, swing2LimitDegrees=10, twistLimitDegrees=10}
 {bone1Name=dyn_leaf_03, bone2Name=Melee_TwoHanded_Handle, swing1LimitDegrees=25, swing2LimitDegrees=45, twistLimitDegrees=5}
 {bone1Name=dyn_leaf_02, bone2Name=Melee_TwoHanded_Handle, swing1LimitDegrees=25, swing2LimitDegrees=45, twistLimitDegrees=5}
 {bone1Name=dyn_leaf_01, bone2Name=Melee_TwoHanded_Handle, swing1LimitDegrees=25, swing2LimitDegrees=45, twistLimitDegrees=5}
 {bone1Name=dyn_leaf_12, bone2Name=Melee_TwoHanded_Handle, swing1LimitDegrees=25, swing2LimitDegrees=45, twistLimitDegrees=5}
 {bone1Name=dyn_leaf_11, bone2Name=dyn_small_vine_03, swing1LimitDegrees=25, swing2LimitDegrees=45, twistLimitDegrees=5}
 {bone1Name=dyn_leaf_10, bone2Name=dyn_small_vine_02, swing1LimitDegrees=25, swing2LimitDegrees=45, twistLimitDegrees=5}
 {bone1Name=dyn_leaf_09, bone2Name=dyn_small_vine_01, swing1LimitDegrees=25, swing2LimitDegrees=45, twistLimitDegrees=5}
 {bone1Name=dyn_leaf_08, bone2Name=dyn_central_vine_06, swing1LimitDegrees=25, swing2LimitDegrees=45, twistLimitDegrees=5}
 {bone1Name=dyn_leaf_07, bone2Name=dyn_central_vine_05, swing1LimitDegrees=25, swing2LimitDegrees=45, twistLimitDegrees=5}
 {bone1Name=dyn_leaf_06, bone2Name=dyn_central_vine_04, swing1LimitDegrees=25, swing2LimitDegrees=45, twistLimitDegrees=5}
 {bone1Name=dyn_leaf_05, bone2Name=dyn_central_vine_03, swing1LimitDegrees=25, swing2LimitDegrees=45, twistLimitDegrees=5}
 {bone1Name=dyn_leaf_04, bone2Name=dyn_central_vine_02, swing1LimitDegrees=25, swing2LimitDegrees=45, twistLimitDegrees=5}
 {
  bone1Name=Melee_TwoHanded_Handle
  bone2Name=dyn_central_vine_03
  swing1LimitDegrees=45
  swing1Motion=Locked
  swing2LimitDegrees=45
  swing2Motion=Locked
  twistLimitDegrees=45
  twistMotion=Locked
 }
]
