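# Animation Blueprint for the Moss Knight pickaxe that adds physics-driven secondary motion to the weapon's skeleton.
# The output pose comes from a RigidBody node run on the reference pose, using the pickaxe physics asset in b_AR_Root space with overridden gravity and scaled component motion.
# A spring-based Anim Dynamics chain for five side vine bones and a Copy Pose From Mesh node are disconnected from the output.
# The event graph is empty, leaving the gravity and wind variables unused.
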AnimBlueprint /BRCosmetics/Weapons/FORT_Melee/Pickaxe_Moss_Knight/Meshes/Moss_Knight_Pickaxe_AnimBP
target_skeleton=&/Moss_Knight_Skeleton
graphs [
 AnimGraph {nodes [
   n1 = Anim Dynamics - Bone: dyn_side_vine_01(Alpha=1, ExternalForce="0,0,0") {
    angularDampingOverride=0.99
    angularSpringConstant=5
    bAngularSpring=true
    bGravityOverrideInSimSpace=true
    bLinearSpring=true
    bOverrideAngularDamping=true
    bOverrideLinearDamping=true
    gravityScale=0.6
    linearDampingOverride=0.7
    linearSpringConstant=5
    simulationSpace=BoneRelative
    alphaBoolBlend {blendOption=Linear}
    boundBone {boneName=dyn_side_vine_01}
    componentLinearAccScale=<0.53 0.53 0.53>
    componentLinearVelScale=<0.75 0.75 0.75>
    physicsBodyDefinitions [{
      sphereCollisionRadius=0
      boundBone {boneName=dyn_side_vine_01}
      boxExtents=<0.2 0.2 0.2>
      constraintSetup {
       twistAxis=AxisY
       angularLimitsMax=<5 5 5>
       angularLimitsMin=<-5 -5 -5>
       angularTarget=<0 0 0>
       linearAxesMax=<0.5 0.5 0.5>
       linearAxesMin=<-0.5 -0.5 -0.5>
      }
      localJointOffset=<0 0 -0.1>
     }]
    relativeSpaceBone {boneName=b_AR_Root}
   }
   n2 = Anim Dynamics - Bone: dyn_side_vine_02(ComponentPose=@n1, Alpha=1, ExternalForce="0,0,0") {
    angularDampingOverride=0.99
    angularSpringConstant=5
    bAngularSpring=true
    bGravityOverrideInSimSpace=true
    bLinearSpring=true
    bOverrideAngularDamping=true
    bOverrideLinearDamping=true
    gravityScale=0.6
    linearDampingOverride=0.7
    linearSpringConstant=5
    simulationSpace=BoneRelative
    alphaBoolBlend {blendOption=Linear}
    boundBone {boneName=dyn_side_vine_02}
    componentLinearAccScale=<0.53 0.53 0.53>
    componentLinearVelScale=<0.75 0.75 0.75>
    physicsBodyDefinitions [{
      sphereCollisionRadius=0
      boundBone {boneName=dyn_side_vine_02}
      boxExtents=<0.2 0.2 0.2>
      constraintSetup {
       twistAxis=AxisY
       angularLimitsMax=<5 5 5>
       angularLimitsMin=<-5 -5 -5>
       angularTarget=<0 0 0>
       linearAxesMax=<0.5 0.5 0.5>
       linearAxesMin=<-0.5 -0.5 -0.5>
      }
      localJointOffset=<0 0 -0.1>
     }]
    relativeSpaceBone {boneName=b_AR_Root}
   }
   n3 = Anim Dynamics - Bone: dyn_side_vine_03(ComponentPose=@n2, Alpha=1, ExternalForce="0,0,0") {
    angularDampingOverride=0.99
    angularSpringConstant=5
    bAngularSpring=true
    bGravityOverrideInSimSpace=true
    bLinearSpring=true
    bOverrideAngularDamping=true
    bOverrideLinearDamping=true
    gravityScale=0.6
    linearDampingOverride=0.7
    linearSpringConstant=5
    simulationSpace=BoneRelative
    alphaBoolBlend {blendOption=Linear}
    boundBone {boneName=dyn_side_vine_03}
    componentLinearAccScale=<0.53 0.53 0.53>
    componentLinearVelScale=<0.75 0.75 0.75>
    physicsBodyDefinitions [{
      sphereCollisionRadius=0
      boundBone {boneName=dyn_side_vine_03}
      boxExtents=<0.2 0.2 0.2>
      constraintSetup {
       twistAxis=AxisY
       angularLimitsMax=<5 5 5>
       angularLimitsMin=<-5 -5 -5>
       angularTarget=<0 0 0>
       linearAxesMax=<0.5 0.5 0.5>
       linearAxesMin=<-0.5 -0.5 -0.5>
      }
      localJointOffset=<0 0 -0.1>
     }]
    relativeSpaceBone {boneName=b_AR_Root}
   }
   n4 = Anim Dynamics - Bone: dyn_side_vine_04(ComponentPose=@n3, Alpha=1, ExternalForce="0,0,0") {
    angularDampingOverride=0.99
    angularSpringConstant=5
    bAngularSpring=true
    bGravityOverrideInSimSpace=true
    bLinearSpring=true
    bOverrideAngularDamping=true
    bOverrideLinearDamping=true
    gravityScale=0.6
    linearDampingOverride=0.7
    linearSpringConstant=5
    simulationSpace=BoneRelative
    alphaBoolBlend {blendOption=Linear}
    boundBone {boneName=dyn_side_vine_04}
    componentLinearAccScale=<0.53 0.53 0.53>
    componentLinearVelScale=<0.75 0.75 0.75>
    physicsBodyDefinitions [{
      sphereCollisionRadius=0
      boundBone {boneName=dyn_side_vine_04}
      boxExtents=<0.2 0.2 0.2>
      constraintSetup {
       twistAxis=AxisY
       angularLimitsMax=<5 5 5>
       angularLimitsMin=<-5 -5 -5>
       angularTarget=<0 0 0>
       linearAxesMax=<0.5 0.5 0.5>
       linearAxesMin=<-0.5 -0.5 -0.5>
      }
      localJointOffset=<0 0 -0.1>
     }]
    relativeSpaceBone {boneName=b_AR_Root}
   }
   Anim Dynamics - Bone: dyn_side_vine_05(ComponentPose=@n4, Alpha=1, ExternalForce="0,0,0") {
    angularDampingOverride=0.99
    angularSpringConstant=5
    bAngularSpring=true
    bGravityOverrideInSimSpace=true
    bLinearSpring=true
    bOverrideAngularDamping=true
    bOverrideLinearDamping=true
    gravityScale=0.6
    linearDampingOverride=0.7
    linearSpringConstant=5
    simulationSpace=BoneRelative
    alphaBoolBlend {blendOption=Linear}
    boundBone {boneName=dyn_side_vine_05}
    componentLinearAccScale=<0.53 0.53 0.53>
    componentLinearVelScale=<0.75 0.75 0.75>
    physicsBodyDefinitions [{
      sphereCollisionRadius=0
      boundBone {boneName=dyn_side_vine_05}
      boxExtents=<0.2 0.2 0.2>
      constraintSetup {
       twistAxis=AxisY
       angularLimitsMax=<5 5 5>
       angularLimitsMin=<-5 -5 -5>
       angularTarget=<0 0 0>
       linearAxesMax=<0.5 0.5 0.5>
       linearAxesMin=<-0.5 -0.5 -0.5>
      }
      localJointOffset=<0 0 -0.1>
     }]
    relativeSpaceBone {boneName=b_AR_Root}
   }
   Copy Pose From Mesh() {bUseAttachedParent=true}
   n5 = RigidBody(ComponentPose=Local To Component(LocalPose=Local Space Ref Pose())) {
    bClampLinearTranslationLimitToRefPose=true
    bOverrideWorldGravity=true
    overridePhysicsAsset=&/Moss_Knight_Pickaxe_Physics
    simulationSpace=BaseBoneSpace
    alphaBoolBlend {blendOption=Linear}
    baseBoneRef {boneName=b_AR_Root}
    componentAppliedLinearAccClamp=<5000 5000 5000>
    componentLinearAccScale=<0.4 0.4 0.5>
    componentLinearVelScale=<0.4 0.4 0.5>
    overrideWorldGravity {z=-1100}
   }
   Output Pose(Result=Component To Local(ComponentPose=@n5)) {layerGroup=None}
  ]}
 EventGraph {links [], nodes []}
]
variables [
 finalGravity {type=Vector}
 windAmbientVector {type=Vector}
 windVector {type=Vector}
]
